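; BP_PalletJack_Window_A_C_B is a window Blueprint from the PalletJack environment set in WildEstate.
; Its construction script passes the actor itself to the SDP ValidateGlassProviderComponents function, which by its name checks the glass provider components.
; The file shows no other logic, so any glass behavior lies outside this graph.
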(fn ConstructionScript
  (SDP|ValidateGlassProviderComponents self))
